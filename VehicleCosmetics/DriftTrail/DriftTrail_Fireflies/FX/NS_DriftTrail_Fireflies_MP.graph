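# Niagara system for the Fireflies drift trail vehicle cosmetic, emitting glowing green fireflies, ribbons and wheel wraps from skeletal-mesh wheel bones.
# Cone fireflies wobble on randomized sine offsets in three spawn groups, while faster sparks and ribbons stream backward with fading inherited movement.
# Wrap emitters scale meshes to each wheel and orbit sprites around it.
# Tire FX masks gate emitters per wheel; user parameters supply offsets and front-end preview tweaks.
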NiagaraSystem /VehicleCosmetics/DriftTrail/DriftTrail_Fireflies/FX/NS_DriftTrail_Fireflies_MP
dependencies {
 usedDataInterfaces [
  &NiagaraDataInterfaceCurve
  &NiagaraDataInterfaceArrayFloat3
  &NiagaraDataInterfaceArrayColor
  &NiagaraDataInterfaceSkeletalMesh
 ]
 usedDynamicInputs []
 usedModules [
  &self:SystemSpawnScript.SystemScriptSource.SystemScriptGraph.NiagaraNodeAssignment_0.SetVariables_#1
  &SystemState
  &FN_UnpackVehicleFlags
  /Game/Effects/Niagara/Modules/Vehicle/FN_UpdateVehicleAttribs
  /Game/Effects/Niagara/Modules/Vehicle/FN_VehicleFlags_Cheatcodes
 ]
}
emitters {
 NE_ConeFireflies {
  emitterUpdateScript [
   &EmitterState()
   &SpawnRate(SpawnRate=45, Use Spawn Probability=false)
   &SpawnRate(SpawnGroup=1, SpawnRate=45, Use Spawn Probability=false) {module=SpawnRate001}
   &SpawnRate(SpawnGroup=2, SpawnRate=10, Use Spawn Probability=false) {module=SpawnRate002}
  ]
  particleSpawnScript [
   &VC_PaintedVariantsOverride() {enabled=false}
   /SetVariables(Particles.RenderVisTag:NiagaraInt32=&SpawnGroup()) {isSetParameters=true}
   &InitializeParticle(
    Color Mode=::NewEnumerator0
    Lifetime=&RandomRangeFloat(Minimum=1, Maximum=1.5)
    Sprite Rotation Mode=::NewEnumerator2
    Sprite Size Mode=::NewEnumerator3
    Uniform Sprite Size=/VehicleCosmetics/FX/Niagara/Modules/MergeEmitters/VC_FloatBySpawnGroup(
     SpawnGroup or ExecIndex:NiagaraInt32=0
     CustomInt:NiagaraInt32=0
     MaximumSpawnGroups:NiagaraInt32=3
     Float SG 0:NiagaraFloat=4
     Float SG 1:NiagaraFloat=6
     Float SG 2:NiagaraFloat=13
     Float SG 3:NiagaraFloat=0
     Float SG 4:NiagaraFloat=0
     Float SG 5:NiagaraFloat=0
    )
    UsePositionOffset=false
    Write Material Random=true
   )
   &PartitionParticles(Partition Mode:ENiagara_PartitionMode=::NewEnumerator5)
   &SkeletalMeshLocation(
    Bone / Socket Index=&Add_Integer(A=$Particles.PartitionParticles.Partition, B=4)
    Bone / Socket Sampling Mode=::NewEnumerator5
    Interpolate Mesh Sampling=false
    MaskSpawnGroup=false
    Mesh Sampling Type=::NewEnumerator6
    Sampled Position Offset=&AddVectorToPosition(Position=<0 0 0>, Vector=<0 0 0>)
    Skeletal Mesh=$System.SkeletalMesh
   )
   &ShapeLocation(MaskSpawnGroup=false, Override Seed=false, Sphere Radius=1, UseOwnerScale=true) {enabled=false}
   &AddVelocity(Velocity=&VC_Mask_Vectors(
     FrontEnd Value:Vector3f=<-1500 0 0>
     blsFrontEndPreview:NiagaraBool=false
     SpeedMin Value:Vector3f=<-1000 0 0>
     SpeedMax Value:Vector3f=<-1000 0 0>
     Speed Alpha:NiagaraFloat=$System.Speed_Alpha
    )) {enabled=false}
   &Color(Color=/VehicleCosmetics/FX/Niagara/Modules/MergeEmitters/VC_Dynamic_ColorBySpawnGroup(
     Color SG 0:LinearColor=<rgba 0.3 1 0.01 1>
     MaximumSpawnGroups:NiagaraInt32=3
     Color SG 1:LinearColor=<rgba 0.6564 1 0 1>
     Color SG 2:LinearColor=<rgba 0.4941 1 0 1>
     Color SG 3:LinearColor=<rgba 1 1 1 1>
     Color SG 4:LinearColor=<rgba 1 1 1 1>
    ))
   /SetVariables(Particles.ColorBackFill:LinearColor=<rgba 0.008 0.0232 0.0012 1>) {isSetParameters=true}
  ]
  particleUpdateScript [
   &FN_DisableTireFX(
    Mask Index Attribute:NiagaraInt32=$Particles.PartitionParticles.Partition
    Mask Index FL=2
    Mask Index FR=3
    Mask Index RL=0
    Mask Index RR=1
   )
   &ShapeLocation(
    MaskSpawnGroup=false
    Offset=/VehicleCosmetics/FX/Niagara/Modules/MergeEmitters/VC_VectorBySpawnGroup(
     SpawnGroup or ExecIndex:NiagaraInt32=0
     CustomInt:NiagaraInt32=0
     MaximumSpawnGroups:NiagaraInt32=3
     Vector Group 0:Vector3f=&MakeVector(
      X=&Add_Float(
       A=&Add_Float(
        A=&MakeFloatFromVector(Vector=$User.offset, Channel=::NewEnumerator0)
        B=&MakeFloatFromVector(Vector=$User.FirefliesOffset, Channel=::NewEnumerator0)
       )
       B=&Sine(
        Normalized Angle=&Add_Float(A=$Emitter.Age, B=&RandomRangeFloat(Minimum=0, Maximum=50))
        Period=&RandomRangeFloat(Minimum=4, Maximum=7)
        Scale=&Multiply_Float(
         A=&RandomRangeFloat(Minimum=20, Maximum=30)
         B=&FloatFromCurve(FloatCurve=DataInterface() {propertyValues {Curve {keys [
              {tangentMode=RCTM_User, time=0, value=1}
              {arriveTangent=-0.0405, interpMode=RCIM_Cubic, leaveTangent=-0.0405, tangentMode=RCTM_User, time=0.5, value=1}
              {arriveTangent=2.1136, interpMode=RCIM_Cubic, leaveTangent=2.1136, tangentMode=RCTM_User, time=1, value=1.75}
             ]}}})
        )
        Bias=0
       )
      )
      Y=&Add_Float(
       A=&Add_Float(
        A=&MakeFloatFromVector(Vector=$User.offset, Channel=::NewEnumerator1)
        B=&MakeFloatFromVector(Vector=$User.FirefliesOffset, Channel=::NewEnumerator1)
       )
       B=&Sine(
        Normalized Angle=&Add_Float(A=$Emitter.Age, B=&RandomRangeFloat(Minimum=0, Maximum=50))
        Period=&RandomRangeFloat(Minimum=4, Maximum=7)
        Scale=&Multiply_Float(
         A=&RandomRangeFloat(Minimum=20, Maximum=30)
         B=&FloatFromCurve(FloatCurve=DataInterface() {propertyValues {Curve {keys [
              {tangentMode=RCTM_User, time=0, value=1}
              {arriveTangent=-0.0405, interpMode=RCIM_Cubic, leaveTangent=-0.0405, tangentMode=RCTM_User, time=0.5, value=1}
              {arriveTangent=2.1136, interpMode=RCIM_Cubic, leaveTangent=2.1136, tangentMode=RCTM_User, time=1, value=1.75}
             ]}}})
        )
        Bias=0
       )
      )
      Z=&Add_Float(
       A=&Add_Float(
        A=&MakeFloatFromVector(Vector=$User.offset, Channel=::NewEnumerator2)
        B=&MakeFloatFromVector(Vector=$User.FirefliesOffset, Channel=::NewEnumerator2)
       )
       B=&Sine(
        Normalized Angle=&Add_Float(A=$Emitter.Age, B=&RandomRangeFloat(Minimum=0, Maximum=50))
        Period=&RandomRangeFloat(Minimum=4, Maximum=7)
        Scale=&Multiply_Float(
         A=&RandomRangeFloat(Minimum=20, Maximum=30)
         B=&FloatFromCurve(FloatCurve=DataInterface() {propertyValues {Curve {keys [
              {tangentMode=RCTM_User, time=0, value=1}
              {arriveTangent=-0.0405, interpMode=RCIM_Cubic, leaveTangent=-0.0405, tangentMode=RCTM_User, time=0.5, value=1}
              {arriveTangent=2.1136, interpMode=RCIM_Cubic, leaveTangent=2.1136, tangentMode=RCTM_User, time=1, value=1.75}
             ]}}})
        )
        Bias=0
       )
      )
     )
     Vector Group 1:Vector3f=&MakeVector(
      X=&Add_Float(
       A=&Add_Float(
        A=&MakeFloatFromVector(Vector=$User.offset, Channel=::NewEnumerator0)
        B=&MakeFloatFromVector(Vector=$User.FirefliesOffset, Channel=::NewEnumerator0)
       )
       B=&Sine(
        Normalized Angle=&Add_Float(A=$Emitter.Age, B=&RandomRangeFloat(Minimum=0, Maximum=50))
        Period=&RandomRangeFloat(Minimum=1, Maximum=3)
        Scale=&Multiply_Float(
         A=&RandomRangeFloat(Minimum=10, Maximum=20)
         B=&FloatFromCurve(FloatCurve=DataInterface() {propertyValues {Curve {keys [
              {tangentMode=RCTM_User, time=0, value=1}
              {arriveTangent=-0.0405, interpMode=RCIM_Cubic, leaveTangent=-0.0405, tangentMode=RCTM_User, time=0.5, value=1}
              {arriveTangent=-4.2376, interpMode=RCIM_Cubic, leaveTangent=-4.2376, tangentMode=RCTM_User, time=1, value=0.25}
             ]}}})
        )
        Bias=0
       )
      )
      Y=&Add_Float(
       A=&Add_Float(
        A=&MakeFloatFromVector(Vector=$User.offset, Channel=::NewEnumerator1)
        B=&MakeFloatFromVector(Vector=$User.FirefliesOffset, Channel=::NewEnumerator1)
       )
       B=&Sine(
        Normalized Angle=&Add_Float(A=$Emitter.Age, B=&RandomRangeFloat(Minimum=0, Maximum=50))
        Period=&RandomRangeFloat(Minimum=1, Maximum=3)
        Scale=&Multiply_Float(
         A=&RandomRangeFloat(Minimum=10, Maximum=20)
         B=&FloatFromCurve(FloatCurve=DataInterface() {propertyValues {Curve {keys [
              {tangentMode=RCTM_User, time=0, value=1}
              {arriveTangent=-0.0405, interpMode=RCIM_Cubic, leaveTangent=-0.0405, tangentMode=RCTM_User, time=0.5, value=1}
              {arriveTangent=-4.2376, interpMode=RCIM_Cubic, leaveTangent=-4.2376, tangentMode=RCTM_User, time=1, value=0.25}
             ]}}})
        )
        Bias=0
       )
      )
      Z=&Add_Float(
       A=&Add_Float(
        A=&MakeFloatFromVector(Vector=$User.offset, Channel=::NewEnumerator2)
        B=&MakeFloatFromVector(Vector=$User.FirefliesOffset, Channel=::NewEnumerator2)
       )
       B=&Sine(
        Normalized Angle=&Add_Float(A=$Emitter.Age, B=&RandomRangeFloat(Minimum=0, Maximum=50))
        Period=&RandomRangeFloat(Minimum=1, Maximum=3)
        Scale=&Multiply_Float(
         A=&RandomRangeFloat(Minimum=10, Maximum=20)
         B=&FloatFromCurve(FloatCurve=DataInterface() {propertyValues {Curve {keys [
              {tangentMode=RCTM_User, time=0, value=1}
              {arriveTangent=-0.0405, interpMode=RCIM_Cubic, leaveTangent=-0.0405, tangentMode=RCTM_User, time=0.5, value=1}
              {arriveTangent=-4.2376, interpMode=RCIM_Cubic, leaveTangent=-4.2376, tangentMode=RCTM_User, time=1, value=0.25}
             ]}}})
        )
        Bias=0
       )
      )
     )
     Vector Group 2:Vector3f=&MakeVector(
      X=&Add_Float(
       A=&Add_Float(
        A=&MakeFloatFromVector(Vector=$User.offset, Channel=::NewEnumerator0)
        B=&MakeFloatFromVector(Vector=$User.FirefliesOffset, Channel=::NewEnumerator0)
       )
       B=&Sine(
        Normalized Angle=&Add_Float(A=$Emitter.Age, B=&RandomRangeFloat(Minimum=0, Maximum=50))
        Period=0.25
        Scale=&RandomRangeFloat(Minimum=1.5, Maximum=2.5)
        Bias=0
       )
      )
      Y=&Add_Float(
       A=&Add_Float(
        A=&MakeFloatFromVector(Vector=$User.offset, Channel=::NewEnumerator1)
        B=&MakeFloatFromVector(Vector=$User.FirefliesOffset, Channel=::NewEnumerator1)
       )
       B=&Sine(
        Normalized Angle=&Add_Float(A=$Emitter.Age, B=&RandomRangeFloat(Minimum=0, Maximum=50))
        Period=0.25
        Scale=&RandomRangeFloat(Minimum=1.5, Maximum=2.5)
        Bias=0
       )
      )
      Z=&Add_Float(
       A=&Add_Float(
        A=&MakeFloatFromVector(Vector=$User.offset, Channel=::NewEnumerator2)
        B=&MakeFloatFromVector(Vector=$User.FirefliesOffset, Channel=::NewEnumerator2)
       )
       B=&Sine(
        Normalized Angle=&Add_Float(A=$Emitter.Age, B=&RandomRangeFloat(Minimum=0, Maximum=50))
        Period=0.25
        Scale=&RandomRangeFloat(Minimum=1.5, Maximum=2.5)
        Bias=0
       )
      )
     )
     Vector Group 3:Vector3f=<0 0 0>
     Vector Group 4:Vector3f=<0 0 0>
     Vector Group 5:Vector3f=<0 0 0>
    )
    Offset Coordinate Space=::Local
    Offset Mode=::NewEnumerator0
    Override Seed=false
    Radius Position:NiagaraFloat=$Particles.Age
    Ring / Disc Distribution Mode:ENiagaraSphereDistributionMode=::NewEnumerator1
    Ring Radius=0
    Rotation Mode=::NewEnumerator2
    Shape Origin=$Particles.Initial.Position
    Shape Primitive=::NewEnumerator4
    UseOwnerScale=true
   ) {module=ShapeLocation001}
   &ShapeLocation(
    MaskSpawnGroup=false
    Offset=&Multiply_VectorByFloat(
     Vector=&MakeVector(
      X=&Add_Float(
       A=&Add_Float(
        A=&MakeFloatFromVector(Vector=$User.offset, Channel=::NewEnumerator0)
        B=&MakeFloatFromVector(Vector=$User.FirefliesOffset, Channel=::NewEnumerator0)
       )
       B=&Sine(
        Normalized Angle=&Add_Float(A=$Emitter.Age, B=&RandomRangeFloat(Minimum=0, Maximum=50))
        Period=&RandomRangeFloat(Minimum=1, Maximum=3)
        Scale=&Multiply_Float(
         A=&RandomRangeFloat(Minimum=10, Maximum=20)
         B=&FloatFromCurve(FloatCurve=DataInterface() {propertyValues {Curve {keys [
              {tangentMode=RCTM_User, time=0, value=1}
              {arriveTangent=-0.0405, interpMode=RCIM_Cubic, leaveTangent=-0.0405, tangentMode=RCTM_User, time=0.5, value=1}
              {arriveTangent=-4.2376, interpMode=RCIM_Cubic, leaveTangent=-4.2376, tangentMode=RCTM_User, time=1, value=0.25}
             ]}}})
        )
        Bias=0
       )
      )
      Y=&Add_Float(
       A=&Add_Float(
        A=&MakeFloatFromVector(Vector=$User.offset, Channel=::NewEnumerator1)
        B=&MakeFloatFromVector(Vector=$User.FirefliesOffset, Channel=::NewEnumerator1)
       )
       B=&Sine(
        Normalized Angle=&Add_Float(A=$Emitter.Age, B=&RandomRangeFloat(Minimum=0, Maximum=50))
        Period=&RandomRangeFloat(Minimum=1, Maximum=3)
        Scale=&Multiply_Float(
         A=&RandomRangeFloat(Minimum=10, Maximum=20)
         B=&FloatFromCurve(FloatCurve=DataInterface() {propertyValues {Curve {keys [
              {tangentMode=RCTM_User, time=0, value=1}
              {arriveTangent=-0.0405, interpMode=RCIM_Cubic, leaveTangent=-0.0405, tangentMode=RCTM_User, time=0.5, value=1}
              {arriveTangent=-4.2376, interpMode=RCIM_Cubic, leaveTangent=-4.2376, tangentMode=RCTM_User, time=1, value=0.25}
             ]}}})
        )
        Bias=0
       )
      )
      Z=&Add_Float(
       A=&Add_Float(
        A=&MakeFloatFromVector(Vector=$User.offset, Channel=::NewEnumerator2)
        B=&MakeFloatFromVector(Vector=$User.FirefliesOffset, Channel=::NewEnumerator2)
       )
       B=&Sine(
        Normalized Angle=&Add_Float(A=$Emitter.Age, B=&RandomRangeFloat(Minimum=0, Maximum=50))
        Period=&RandomRangeFloat(Minimum=1, Maximum=3)
        Scale=&Multiply_Float(
         A=&RandomRangeFloat(Minimum=10, Maximum=20)
         B=&FloatFromCurve(FloatCurve=DataInterface() {propertyValues {Curve {keys [
              {tangentMode=RCTM_User, time=0, value=1}
              {arriveTangent=-0.0405, interpMode=RCIM_Cubic, leaveTangent=-0.0405, tangentMode=RCTM_User, time=0.5, value=1}
              {arriveTangent=-4.2376, interpMode=RCIM_Cubic, leaveTangent=-4.2376, tangentMode=RCTM_User, time=1, value=0.25}
             ]}}})
        )
        Bias=0
       )
      )
     )
     Float=&VC_FloatBySpawnGroup(
      SpawnGroup or ExecIndex:NiagaraInt32=0
      CustomInt:NiagaraInt32=0
      MaximumSpawnGroups:NiagaraInt32=3
      Float SG 0:NiagaraFloat=0
      Float SG 1:NiagaraFloat=1
      Float SG 2:NiagaraFloat=0
      Float SG 3:NiagaraFloat=0
      Float SG 4:NiagaraFloat=0
      Float SG 5:NiagaraFloat=0
     )
    )
    Offset Coordinate Space=::Local
    Offset Mode=::NewEnumerator0
    Override Seed=false
    Radius Position:NiagaraFloat=$Particles.Age
    Ring / Disc Distribution Mode:ENiagaraSphereDistributionMode=::NewEnumerator1
    Ring Radius=1
    Rotation Mode=::NewEnumerator2
    Shape Origin=$Particles.Initial.Position
    Shape Primitive=::NewEnumerator4
    UseOwnerScale=true
   ) {enabled=false, module=ShapeLocation002}
   &ParticleState()
   &ScaleSpriteSize(Uniform Curve Sprite Scale=DataInterface() {propertyValues {Curve {keys [
        {interpMode=RCIM_Cubic, tangentMode=RCTM_User, time=0, value=0.4}
        {interpMode=RCIM_Cubic, tangentMode=RCTM_User, time=0.1, value=1}
        {arriveTangent=-0.0064, interpMode=RCIM_Cubic, leaveTangent=-0.0064, tangentMode=RCTM_User, time=0.3, value=1}
        {interpMode=RCIM_Cubic, tangentMode=RCTM_User, time=1, value=0}
       ]}}})
   &ScaleColor(Scale RGB=&VectorFromFloat(Value=5), ScaleA=true, ScaleRGB=true)
   &RotateAroundPoint(
    Radius=2
    Rotation Center:NiagaraPosition=$Particles.Initial.Position
    Rotation Rate:NiagaraFloat=&SelectFloatFromArray(
     Array Sampling Mode=::NewEnumerator0
     Direct Array Index=?
     Float Selection Array=DataInterface() {propertyValues {FloatData [-1,1]}}
     Random Seed=0
    )
    X Axis:Vector3f=&SelectVectorFromArray(
     Array Sampling Mode=::NewEnumerator0
     Direct Array Index=?
     Vector Selection Array=DataInterface() {propertyValues {FloatData [<0.5 0 0>, <0 0.5 0>]}}
     Random Seed=0
    )
    Y Axis:Vector3f=<0 0 1>
   ) {enabled=false}
   &DynamicMaterialParameters(
    Index 0 Param 2=0.3
    Index 0 Param 3=&FloatFromCurve(FloatCurve=DataInterface() {propertyValues {Curve {keys [{interpMode=RCIM_Cubic, tangentMode=RCTM_User, time=0, value=0}, {interpMode=RCIM_Cubic, tangentMode=RCTM_User, time=0.1, value=1}]}}})
    Module.Index0Param0Write=true
    Module.Index0Param1Write=true
    Module.Index0Param2Write=true
    Module.Index0Param3Write=true
   )
   &SolveForcesAndVelocity(Clamp Velocity=false, Limit Acceleration=false)
   /SetVariables(Particles.ScaleColorBackFill:LinearColor=&MultiplyLinearColor(
     Linear Color=$Particles.ColorBackFill
     Scale Factor=&MakeVector4_FromFloat(Float=&FloatFromCurve(FloatCurve=DataInterface() {propertyValues {Curve {keys [{interpMode=RCIM_Cubic, tangentMode=RCTM_User, time=0, value=0}, {interpMode=RCIM_Cubic, tangentMode=RCTM_User, time=0.1, value=1}]}}}))
    )) {isSetParameters=true}
  ]
  rendererClasses [&NiagaraSpriteRendererProperties]
  renderers [
   {rendererClass=&NiagaraSpriteRendererProperties}
   {rendererClass=&NiagaraSpriteRendererProperties, rendererIndex=1}
   {rendererClass=&NiagaraSpriteRendererProperties, rendererIndex=2}
   {rendererClass=&NiagaraSpriteRendererProperties, rendererIndex=3}
   {rendererClass=&NiagaraSpriteRendererProperties, rendererIndex=4}
   {rendererClass=&NiagaraSpriteRendererProperties, rendererIndex=5}
  ]
 }
 NE_TrailFireflies001 {
  emitterUpdateScript [
   &EmitterState()
   &SpawnRate(SpawnRate=180, Use Spawn Probability=false)
  ]
  particleSpawnScript [
   &VC_PaintedVariantsOverride() {enabled=false}
   /SetVariables(Particles.RenderVisTag:NiagaraInt32=&SpawnGroup()) {isSetParameters=true}
   &InitializeParticle(
    Color Mode=::NewEnumerator0
    Lifetime=&RandomRangeFloat(Minimum=0.25, Maximum=0.4)
    Sprite Rotation Mode=::NewEnumerator2
    Sprite Size Mode=::NewEnumerator1
    Uniform Sprite Size Max=6
    Uniform Sprite Size Min=4
    UsePositionOffset=false
    Write Material Random=true
   )
   &PartitionParticles()
   &SkeletalMeshLocation(
    Bone / Socket Index=&Add_Integer(A=$Particles.PartitionParticles.Partition, B=4)
    Bone / Socket Sampling Mode=::NewEnumerator5
    Interpolate Mesh Sampling=false
    MaskSpawnGroup=false
    Mesh Sampling Type=::NewEnumerator6
    Sampled Position Offset=&AddVectorToPosition(Position=<0 0 0>, Vector=<0 0 0>)
    Skeletal Mesh=$System.SkeletalMesh
   )
   &ShapeLocation(
    MaskSpawnGroup=false
    Offset=&Add_Vector(A=$User.offset, B=$User.FirefliesOffset)
    Offset Coordinate Space=::Local
    Offset Mode=::NewEnumerator0
    Override Seed=false
    Shape Origin=$Particles.Position
    Sphere Radius=20
    UseOwnerScale=true
   )
   &AddVelocity(Velocity=<-1000 0 0>)
   &Color(Color=&SelectLinearColorFromArray(
     Array Sampling Mode=::NewEnumerator0
     Direct Array Index=?
     Color Selection Array=DataInterface() {propertyValues {ColorData [<rgba 0.3 1 0.01 1>, <rgba 0.6564 1 0 1>]}}
     Random Seed=0
    ))
   /SetVariables(Particles.ColorBackFill:LinearColor=<rgba 0.008 0.0232 0.0012 1>) {isSetParameters=true}
  ]
  particleUpdateScript [
   &FN_DisableTireFX(
    Mask Index Attribute:NiagaraInt32=$Particles.PartitionParticles.Partition
    Mask Index FL=2
    Mask Index FR=3
    Mask Index RL=0
    Mask Index RR=1
   )
   &ShapeLocation(
    MaskSpawnGroup=false
    Offset=&MakeVector(
     X=0
     Y=&Add_Float(
      A=&Add_Float(
       A=&MakeFloatFromVector(Vector=$User.offset, Channel=::NewEnumerator1)
       B=&MakeCustomFloatFromBool(Bool=$User.bIsFrontendPreview, True Float=0, False Float=0)
      )
      B=&Sine(
       Normalized Angle=&Add_Float(A=$Emitter.Age, B=&RandomRangeFloat(Minimum=0, Maximum=50))
       Period=&RandomRangeFloat(Minimum=1, Maximum=3)
       Scale=&Multiply_Float(
        A=&RandomRangeFloat(Minimum=10, Maximum=20)
        B=&FloatFromCurve(FloatCurve=DataInterface() {propertyValues {Curve {keys [
             {tangentMode=RCTM_User, time=0, value=1}
             {arriveTangent=-0.0405, interpMode=RCIM_Cubic, leaveTangent=-0.0405, tangentMode=RCTM_User, time=0.5, value=1}
             {arriveTangent=-4.2376, interpMode=RCIM_Cubic, leaveTangent=-4.2376, tangentMode=RCTM_User, time=1, value=0.25}
            ]}}})
       )
       Bias=0
      )
     )
     Z=&Add_Float(
      A=&Add_Float(
       A=&MakeFloatFromVector(Vector=$User.offset, Channel=::NewEnumerator2)
       B=&MakeCustomFloatFromBool(Bool=$User.bIsFrontendPreview, True Float=10, False Float=0)
      )
      B=&Sine(
       Normalized Angle=&Add_Float(A=$Emitter.Age, B=&RandomRangeFloat(Minimum=0, Maximum=50))
       Period=&RandomRangeFloat(Minimum=1, Maximum=3)
       Scale=&Multiply_Float(
        A=&RandomRangeFloat(Minimum=10, Maximum=20)
        B=&FloatFromCurve(FloatCurve=DataInterface() {propertyValues {Curve {keys [
             {tangentMode=RCTM_User, time=0, value=1}
             {arriveTangent=-0.0405, interpMode=RCIM_Cubic, leaveTangent=-0.0405, tangentMode=RCTM_User, time=0.5, value=1}
             {arriveTangent=-4.2376, interpMode=RCIM_Cubic, leaveTangent=-4.2376, tangentMode=RCTM_User, time=1, value=0.25}
            ]}}})
       )
       Bias=0
      )
     )
    )
    Offset Coordinate Space=::Local
    Offset Mode=::NewEnumerator0
    Override Seed=false
    Radius Position:NiagaraFloat=$Particles.Age
    Ring / Disc Distribution Mode:ENiagaraSphereDistributionMode=::NewEnumerator1
    Ring Radius=1
    Rotation Mode=::NewEnumerator2
    Shape Origin=$Particles.Initial.Position
    Shape Primitive=::NewEnumerator4
    UseOwnerScale=true
   ) {enabled=false, module=ShapeLocation001}
   &InheritSourceMovement(
    Applied Position Scale=&Multiply_VectorByFloat(
     Vector=<1 1 1>
     Float=&FloatFromCurve(FloatCurve=DataInterface() {propertyValues {Curve {keys [
          {
           arriveTangentWeight=0.4252
           interpMode=RCIM_Cubic
           leaveTangentWeight=0.4252
           tangentMode=RCTM_User
           tangentWeightMode=RCTWM_WeightedBoth
           time=0.15
           value=1
          }
          {interpMode=RCIM_Cubic, tangentMode=RCTM_User, time=1, value=0}
         ]}}})
    )
    ApplyForce=false
    ApplyPosition=true
    ApplyPositionFalloff=false
   )
   &ParticleState()
   &ScaleSpriteSize(Uniform Curve Sprite Scale=DataInterface() {propertyValues {Curve {keys [
        {interpMode=RCIM_Cubic, tangentMode=RCTM_User, time=0, value=0.4}
        {interpMode=RCIM_Cubic, tangentMode=RCTM_User, time=0.1, value=1}
        {arriveTangent=0.0864, interpMode=RCIM_Cubic, leaveTangent=0.0864, tangentMode=RCTM_User, time=0.65, value=1}
        {interpMode=RCIM_Cubic, tangentMode=RCTM_User, time=1, value=0}
       ]}}})
   &ScaleColor(Scale RGB=&VectorFromFloat(Value=5), ScaleA=true, ScaleRGB=true)
   &RotateAroundPoint(
    Radius=2
    Rotation Center:NiagaraPosition=$Particles.Initial.Position
    Rotation Rate:NiagaraFloat=&SelectFloatFromArray(
     Array Sampling Mode=::NewEnumerator0
     Direct Array Index=?
     Float Selection Array=DataInterface() {propertyValues {FloatData [-1,1]}}
     Random Seed=0
    )
    X Axis:Vector3f=&SelectVectorFromArray(
     Array Sampling Mode=::NewEnumerator0
     Direct Array Index=?
     Vector Selection Array=DataInterface() {propertyValues {FloatData [<0.5 0 0>, <0 0.5 0>]}}
     Random Seed=0
    )
    Y Axis:Vector3f=<0 0 1>
   ) {enabled=false}
   &DynamicMaterialParameters(
    Index 0 Param 2=0.3
    Index 0 Param 3=&FloatFromCurve(FloatCurve=DataInterface() {propertyValues {Curve {keys [{interpMode=RCIM_Cubic, tangentMode=RCTM_User, time=0, value=0}, {interpMode=RCIM_Cubic, tangentMode=RCTM_User, time=0.1, value=1}]}}})
    Module.Index0Param0Write=true
    Module.Index0Param1Write=true
    Module.Index0Param2Write=true
    Module.Index0Param3Write=true
   ) {enabled=false}
   &SolveForcesAndVelocity(Clamp Velocity=false, Limit Acceleration=false)
   /SetVariables(Particles.ScaleColorBackFill:LinearColor=&MultiplyLinearColor(
     Linear Color=$Particles.ColorBackFill
     Scale Factor=&MakeVector4_FromFloat(Float=&FloatFromCurve(FloatCurve=DataInterface() {propertyValues {Curve {keys [{interpMode=RCIM_Cubic, tangentMode=RCTM_User, time=0, value=0}, {interpMode=RCIM_Cubic, tangentMode=RCTM_User, time=0.1, value=1}]}}}))
    )) {isSetParameters=true}
  ]
  rendererClasses [&NiagaraSpriteRendererProperties]
  renderers [
   {rendererClass=&NiagaraSpriteRendererProperties}
   {rendererClass=&NiagaraSpriteRendererProperties, rendererIndex=1}
  ]
 }
 NE_TrailRibbon {
  emitterUpdateScript [
   &EmitterState(Life Cycle Mode=::NewEnumerator1, Loop Behavior=::NewEnumerator1, Loop Duration Mode=::NewEnumerator1, UseLoopDelay=false)
   &VC_RateQualityHandler(Epic Spawn Rate=90, High Spawn Rate=60, Low Spawn Rate=30, Medium Spawn Rate=45, Use Custom Rate?=true)
   &VC_SpawnRibbon_v1(Number of Ribbons=2, SpawnRate=40)
  ]
  particleSpawnScript [
   &VC_PaintedVariantsOverride() {enabled=false}
   &VC_InitializeRibbon_v1(
    Custom Offset=true
    Enable Skeletal Position=true
    Lifetime=&MakeCustomFloatFromBool(Bool=$User.bIsFrontendPreview, True Float=1, False Float=1)
    Local Vector Offset=&MakeCustomVectorFromBool(
     Bool=$User.bIsFrontendPreview
     True Vector=&Add_Vector(A=$User.offset, B=<-30 0 25>)
     False Vector=&Add_Vector(A=$User.offset, B=<-30 0 25>)
    )
    Offset Amount=1
    Ribbon Width:NiagaraFloat=&MakeCustomFloatFromBool(Bool=$User.bIsFrontendPreview, True Float=40, False Float=60)
   )
   &AddVelocity(Velocity=<-1000 0 0>) {module=AddVelocity001}
   &Color(Color=&MakeCustomLinearColorFromBool(Bool=$User.bIsFrontendPreview, True Color=<rgba 0 0.25 0 0.5>, False Color=<rgba 0.1 1 0.1 0.2>))
  ]
  particleUpdateScript [
   &FN_DisableTireFX(Mask Index Attribute:NiagaraInt32=$Particles.SpawnGroup, Mask Index FL=2, Mask Index FR=3, Mask Index RL=0, Mask Index RR=1)
   &InheritSourceMovement(
    Applied Position Scale=&Multiply_VectorByFloat(
     Vector=<1 1 1>
     Float=&FloatFromCurve(FloatCurve=DataInterface() {propertyValues {Curve {keys [
          {
           arriveTangentWeight=0.4252
           interpMode=RCIM_Cubic
           leaveTangentWeight=0.4252
           tangentMode=RCTM_User
           tangentWeightMode=RCTWM_WeightedBoth
           time=0.15
           value=1
          }
          {interpMode=RCIM_Cubic, tangentMode=RCTM_User, time=1, value=0}
         ]}}})
    )
    ApplyForce=false
    ApplyPosition=true
    ApplyPositionFalloff=false
   )
   &ParticleState()
   &SolveForcesAndVelocity(Clamp Velocity=false, Limit Acceleration=false)
   &ScaleColor(
    Scale Alpha=&FloatFromCurve(FloatCurve=DataInterface() {propertyValues {Curve {keys [<at 0 0>,<at 0.05 1>,<at 1 0>]}}}, Scale Curve=1.75)
    Scale RGB=&VectorFromFloat(Value=10)
    ScaleA=true
    ScaleRGB=true
   )
   &VC_UpdateRibbon_v1(Enable Skeletal Position=true)
  ]
  rendererClasses [&NiagaraRibbonRendererProperties]
  renderers [{rendererClass=&NiagaraRibbonRendererProperties}]
 }
 NE_VC_DriftTrail_WheelWrap_Parent {
  emitterUpdateScript [
   &EmitterState(
    Inactive Response=::NewEnumerator1
    Life Cycle Mode=::NewEnumerator1
    Loop Behavior=::NewEnumerator1
    Loop Duration Mode=::NewEnumerator1
    UseLoopDelay=false
   )
   &SpawnBurst_Instantaneous(Spawn Count=8, Use Spawn Probability=false, UseLoopCountLimit=false)
  ]
  particleSpawnScript [
   &VC_PaintedVariantsOverride() {enabled=false}
   &InitializeParticle(
    Color Mode=::NewEnumerator0
    Lifetime=$Emitter.CurrentLoopDuration
    Mesh Scale Mode=::NewEnumerator4
    Sprite Rotation Mode=::NewEnumerator2
    Sprite Size Mode=::NewEnumerator3
    UsePositionOffset=false
    Write Material Random=true
   )
   &PartitionParticles()
   &SkeletalMeshLocation(
    Bone / Socket Index=&Add_Integer(A=$Particles.PartitionParticles.Partition, B=4)
    Bone / Socket Sampling Mode=::NewEnumerator5
    Interpolate Mesh Sampling=false
    MaskSpawnGroup=false
    Mesh Sampling Type=::NewEnumerator6
    Skeletal Mesh=$System.SkeletalMesh
   )
   &Color(Color=&MakeCustomLinearColorFromBool(Bool=$User.bIsFrontendPreview, True Color=<rgba 0.2222 1 0.2222 0.2>, False Color=<rgba 0.2222 1 0.2222 0.1>))
  ]
  particleUpdateScript [
   &FN_DisableTireFX(
    Mask Index Attribute:NiagaraInt32=$Particles.PartitionParticles.Partition
    Mask Index FL=2
    Mask Index FR=3
    Mask Index RL=0
    Mask Index RR=1
   )
   &VC_WheelWrapScale(Wheel Index Attribute:NiagaraInt32=&Add_Integer(A=$Particles.PartitionParticles.Partition, B=2))
   &ScaleColor(Scale Alpha=0.15, Scale RGB=&VectorFromFloat(Value=10), ScaleA=true, ScaleRGB=true)
   &ParticleState(Kill Particles When Lifetime Has Elapsed=false)
   &UpdateMeshOrientation(Rotation Coordinate Space=::NewEnumerator2, Rotation Vector=<0 1 0>) {enabled=false}
   &SkeletalMeshLocation(
    Bone / Socket Index=&Add_Integer(A=$Particles.PartitionParticles.Partition, B=4)
    Bone / Socket Sampling Mode=::NewEnumerator5
    MaskSpawnGroup=false
    Mesh Sampling Type=::NewEnumerator6
    Skeletal Mesh=$System.SkeletalMesh
   ) {module=SkeletalMeshLocation001}
   &SolveForcesAndVelocity(Clamp Velocity=false, Limit Acceleration=false)
  ]
  rendererClasses [&NiagaraMeshRendererProperties]
  renderers [{rendererClass=&NiagaraMeshRendererProperties, rendererIndex=0}]
 }
 NE_VC_DriftTrail_WheelWrap_Parent001 {
  emitterUpdateScript [
   &EmitterState(
    Inactive Response=::NewEnumerator1
    Life Cycle Mode=::NewEnumerator1
    Loop Behavior=::NewEnumerator1
    Loop Duration Mode=::NewEnumerator1
    UseLoopDelay=false
   )
   &SpawnRate(SpawnRate=70, Use Spawn Probability=false)
   &SpawnBurst_Instantaneous(Spawn Count=40, Use Spawn Probability=false, UseLoopCountLimit=false) {enabled=false}
  ]
  particleSpawnScript [
   &VC_PaintedVariantsOverride() {enabled=false}
   &InitializeParticle(
    Color Mode=::NewEnumerator0
    Mesh Scale Mode=::NewEnumerator4
    Position Offset=<0 250 0>
    Sprite Rotation Mode=::NewEnumerator2
    Sprite Size Mode=::NewEnumerator1
    Uniform Sprite Size Max=6
    Uniform Sprite Size Min=4
    UsePositionOffset=false
    Write Material Random=true
   )
   &PartitionParticles()
   &SkeletalMeshLocation(
    Bone / Socket Index=&Add_Integer(A=$Particles.PartitionParticles.Partition, B=4)
    Bone / Socket Sampling Mode=::NewEnumerator5
    Interpolate Mesh Sampling=false
    MaskSpawnGroup=false
    Mesh Sampling Type=::NewEnumerator6
    Sampled Position Offset=&RandomRangeVector(Minimum=<0 -15 0>, Maximum=<0 15 0>)
    Skeletal Mesh=$System.SkeletalMesh
   )
   &Color(Color=&SelectLinearColorFromArray(
     Array Sampling Mode=::NewEnumerator0
     Direct Array Index=?
     Color Selection Array=DataInterface() {propertyValues {ColorData [<rgba 0.3 1 0.01 1>, <rgba 0.6564 1 0 1>]}}
     Random Seed=0
    ))
  ]
  particleUpdateScript [
   &FN_DisableTireFX(
    Mask Index Attribute:NiagaraInt32=$Particles.PartitionParticles.Partition
    Mask Index FL=2
    Mask Index FR=3
    Mask Index RL=0
    Mask Index RR=1
   )
   &VC_WheelWrapScale(Wheel Index Attribute:NiagaraInt32=&Add_Integer(A=$Particles.PartitionParticles.Partition, B=2))
   &ParticleState()
   &ScaleColor(Scale RGB=&VectorFromFloat(Value=5), ScaleA=true, ScaleRGB=true)
   &DynamicMaterialParameters(
    Index 0 Param 2=0.3
    Index 0 Param 3=&FloatFromCurve(FloatCurve=DataInterface() {propertyValues {Curve {keys [{interpMode=RCIM_Cubic, tangentMode=RCTM_User, time=0, value=0}, {interpMode=RCIM_Cubic, tangentMode=RCTM_User, time=0.1, value=1}]}}})
    Module.Index0Param0Write=true
    Module.Index0Param1Write=true
    Module.Index0Param2Write=true
    Module.Index0Param3Write=true
   )
   &RotateAroundPoint(
    Radius=50
    Rotation Center:NiagaraPosition=$Particles.Initial.Position
    Rotation Phase:NiagaraFloat=&RandomRangeFloat(Minimum=0, Maximum=-0.25)
    Rotation Rate:NiagaraFloat=&RandomRangeFloat(Minimum=-1, Maximum=-2)
    Y Axis:Vector3f=<0 0 1>
   ) {enabled=false}
   &RotateAroundPoint(
    Radius:NiagaraFloat=&Multiply_Float(
     A=&RandomRangeFloat(Minimum=54, Maximum=50)
     B=&MakeFloatFromVector(Vector=$Particles.Scale, Channel=::NewEnumerator0)
    )
    Rotation Center:NiagaraPosition=&MaskVectorBySpawnGroup(
     Masked Vector=&AddVectorToPosition(Position=$StackContext.Position, Vector=&RandomRangeVector(Minimum=<-3 -10 -3>, Maximum=<3 10 3>))
     Passthrough Vector=Linked() {linkedVariable {name=StackContext.Position, type=NiagaraPosition}}
    )
    Rotation Rate:NiagaraFloat=&MaskFloatBySpawnGroup(Masked Float=&RandomRangeFloat(Minimum=-1, Maximum=-0.5), Passthrough Float=0)
    Set Particle Position=false
    Y Axis:Vector3f=<0 0 1>
   ) {module=RotateAroundPoint001}
   &UpdateMeshOrientation(Rotation Coordinate Space=::NewEnumerator2, Rotation Vector=<0 1 0>) {enabled=false}
   /SetVariables(Particles.AroundPosition:NiagaraPosition=$Output.RotateAroundPoint001.OutputPosition) {isSetParameters=true}
   &SkeletalMeshLocation(
    Bone / Socket Index=&Add_Integer(A=$Particles.PartitionParticles.Partition, B=4)
    Bone / Socket Sampling Mode=::NewEnumerator5
    MaskSpawnGroup=false
    Mesh Sampling Type=::NewEnumerator6
    Skeletal Mesh=$System.SkeletalMesh
   ) {module=SkeletalMeshLocation001}
   &SolveForcesAndVelocity(Clamp Velocity=false, Limit Acceleration=false)
   /SetVariables(Particles.BokehBackFill_Color:LinearColor=&MakeCustomLinearColorFromBool(Bool=$User.bIsPainted, True Color=$User.PaintedColor, False Color=<rgba 0.008 0.0284 0 5>)) {isSetParameters=true}
  ]
  rendererClasses [&NiagaraSpriteRendererProperties, &NiagaraMeshRendererProperties]
  renderers [
   {rendererClass=&NiagaraSpriteRendererProperties}
   {rendererClass=&NiagaraSpriteRendererProperties, rendererIndex=1}
   {rendererClass=&NiagaraMeshRendererProperties, rendererIndex=2}
  ]
 }
}
summary {
 systemName=NS_DriftTrail_Fireflies_MP
 emitters [
  {emitterName=NE_TrailRibbon, rendererClasses [&NiagaraRibbonRendererProperties]}
  {emitterName=NE_VC_DriftTrail_WheelWrap_Parent, rendererClasses [&NiagaraMeshRendererProperties]}
  {emitterName=NE_ConeFireflies, rendererClasses [&NiagaraSpriteRendererProperties]}
  {
   emitterName=NE_VC_DriftTrail_WheelWrap_Parent001
   rendererClasses [&NiagaraSpriteRendererProperties, &NiagaraMeshRendererProperties]
  }
  {emitterName=NE_TrailFireflies001, rendererClasses [&NiagaraSpriteRendererProperties]}
 ]
 userVariables [
  User.bIsFrontendPreview {type=NiagaraBool}
  User.bIsPainted {type=NiagaraBool}
  User.ColorVariantID {type=NiagaraInt32}
  User.DriftTrailPos {type=NiagaraPosition, description="", defaultValue {}}
  User.FirefliesOffset {type=Vector3f, description="", defaultValue {struct=Vector3f, value=<-50 0 25>}}
  User.offset {type=Vector3f, description="offset from wheels(set through cosmetic template)", defaultValue {struct=Vector3f, value=<0 0 0>}}
  User.PaintedColor {type=LinearColor, description="", defaultValue {struct=LinearColor, value=<rgba 1 1 1 1>}}
  User.PaintedScalar {type=NiagaraFloat, defaultValue=1, description=""}
  User.VehicleFlags {type=NiagaraInt32, defaultValue=806289408}
  User.Wheel_Radius_Array {
   type=NiagaraDataInterfaceArrayFloat
   defaultValue {
    struct=NiagaraExt_VariableValue_DataInterface
    value {
     dataInterface=&self:NiagaraDataInterfaceArrayFloat_0
     dataInterfaceClass=&NiagaraDataInterfaceArrayFloat
    }
   }
  }
  User.Wheel_Width_Array {
   type=NiagaraDataInterfaceArrayFloat
   defaultValue {
    struct=NiagaraExt_VariableValue_DataInterface
    value {
     dataInterface=&self:NiagaraDataInterfaceArrayFloat_1
     dataInterfaceClass=&NiagaraDataInterfaceArrayFloat
    }
   }
  }
  User.WheelSize {type=NiagaraFloat, defaultValue=0.1, description=""}
 ]
}
